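# Physics asset for the medium male Handyman body cosmetic, defining bodies for its dangling left and right clip bones.
# Capsules on spine_02 and spine_03 approximate the torso, and a tiny master body per side links to spine_03 with free angular motion.
# From those masters hang a three-link left chain and a single right body, with tight twist limits and narrow-to-moderate swing limits.
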PhysicsAsset /BRCosmetics/Characters/Player/Male/Medium/Bodies/M_MED_Handyman/Meshes/M_MED_Handyman_Dyn
bodies {
 dyn_clips_l [{shapeType=Capsule, length=10.8829, radius=2, center=<2.5 0 0>, rotation=<rot 0 0 90>}]
 dyn_clips_l_1 [{shapeType=Capsule, length=10.8829, radius=2, center=<2.5 0 0>, rotation=<rot 0 0 90>}]
 dyn_clips_l_2 [{shapeType=Capsule, length=10.8829, radius=2, center=<2.5 0 0>, rotation=<rot 0 0 90>}]
 dyn_clips_l_MSTR [{shapeType=Capsule, length=0.505, radius=0.505, rotation=<rot 0 -90 0>}]
 dyn_clips_r [{shapeType=Capsule, length=10.8829, radius=2, center=<5 0 0>, rotation=<rot 90 0 0>}]
 dyn_clips_r_MSTR [{shapeType=Capsule, length=0.505, radius=0.505, rotation=<rot 0 -90 0>}]
 spine_02 [{shapeType=Capsule, length=4.4894, radius=10.2021, center=<2.6933 1.4652 0.0001>}]
 spine_03 [{shapeType=Capsule, length=6.8793, radius=10.2021, center=<3.652 1.4652 0.0001>}]
}
constraints [
 {bone1Name=dyn_clips_r, bone2Name=dyn_clips_r_MSTR, swing1LimitDegrees=30, swing2LimitDegrees=30, twistLimitDegrees=5}
 {bone1Name=dyn_clips_l, bone2Name=dyn_clips_l_MSTR, swing1LimitDegrees=30, swing2LimitDegrees=30, twistLimitDegrees=5}
 {bone1Name=dyn_clips_l_1, bone2Name=dyn_clips_l, swing1LimitDegrees=5, swing2LimitDegrees=30, twistLimitDegrees=5}
 {bone1Name=dyn_clips_l_2, bone2Name=dyn_clips_l_1, swing1LimitDegrees=5, swing2LimitDegrees=30, twistLimitDegrees=5}
 {
  bone1Name=dyn_clips_r_MSTR
  bone2Name=spine_03
  swing1LimitDegrees=45
  swing1Motion=Free
  swing2LimitDegrees=45
  swing2Motion=Free
  twistLimitDegrees=45
  twistMotion=Free
 }
 {
  bone1Name=dyn_clips_l_MSTR
  bone2Name=spine_03
  swing1LimitDegrees=45
  swing1Motion=Free
  swing2LimitDegrees=45
  swing2Motion=Free
  twistLimitDegrees=45
  twistMotion=Free
 }
 {
  bone1Name=spine_03
  bone2Name=spine_02
  swing1LimitDegrees=45
  swing1Motion=Free
  swing2LimitDegrees=45
  swing2Motion=Free
  twistLimitDegrees=45
  twistMotion=Free
 }
]
