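# Sound cue for the Handyman character's arm closing, classed as weapon foley.
# It mixes one randomly chosen arm-close wave with one randomly chosen target-stop wave, the latter slightly louder, each drawn from three equally weighted variations.
# A Modulator on the mix adds subtle pitch variation.
# The cue plays at reduced volume using weapon-reload attenuation.
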SoundCue /BRCosmetics/Sounds/Characters/Handyman/HandyMan_Arm_Close_Cue
root=Modulator_1
nodes [
 {id=Mixer_1, children [Random_0,Random_1], properties {inputVolume [1,1.2]}}
 {id=Modulator_1, children [Mixer_1], properties {pitchMax=1.02, pitchMin=0.98}}
 {id=Random_0, children [WavePlayer_0,WavePlayer_2,WavePlayer_3], properties {weights [1,1,1]}}
 {id=Random_1, children [WavePlayer_4,WavePlayer_5,WavePlayer_6], properties {weights [1,1,1]}}
 {id=WavePlayer_0, properties {soundWaveAssetPtr=&/HandyMan_Arm_Close_01}}
 {id=WavePlayer_2, properties {soundWaveAssetPtr=&/HandyMan_Arm_Close_02}}
 {id=WavePlayer_3, properties {soundWaveAssetPtr=&/HandyMan_Arm_Close_03}}
 {id=WavePlayer_4, properties {soundWaveAssetPtr=&/HandyMan_Arm_TargetStop_01}}
 {id=WavePlayer_5, properties {soundWaveAssetPtr=&/HandyMan_Arm_TargetStop_02}}
 {id=WavePlayer_6, properties {soundWaveAssetPtr=&/HandyMan_Arm_TargetStop_03}}
]
properties {
 attenuationSettings=&Fort_Weapon_Reload_Att
 maxDistance=2250
 soundClassObject=&Fort_WeaponFoley
 volumeMultiplier=0.251
}
