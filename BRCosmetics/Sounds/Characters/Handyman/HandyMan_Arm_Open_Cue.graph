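# Handyman arm-open sound cue that layers an opening sound, a target-start sound and a quieter looping target sound.
# A mixer combines two branches that each pick randomly from three equally weighted variations with a looping wave at lower volume.
# The open branch runs through a delay node and a pitch-locked modulator, while the target-start branch uses a default modulator.
# It plays in the weapon foley sound class with weapon reload attenuation.
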SoundCue /BRCosmetics/Sounds/Characters/Handyman/HandyMan_Arm_Open_Cue
root=Mixer_0
nodes [
 {id=Delay_0, children [Modulator_1]}
 {id=Mixer_0, children [Delay_0,Modulator_2,WavePlayer_7], properties {inputVolume [1.2,1.2,0.35]}}
 {id=Modulator_1, children [Random_0], properties {pitchMax=1, pitchMin=1}}
 {id=Modulator_2, children [Random_1]}
 {id=Random_0, children [WavePlayer_0,WavePlayer_2,WavePlayer_3], properties {weights [1,1,1]}}
 {id=Random_1, children [WavePlayer_4,WavePlayer_5,WavePlayer_6], properties {weights [1,1,1]}}
 {id=WavePlayer_0, properties {soundWaveAssetPtr=&/HandyMan_Arm_Open_01}}
 {id=WavePlayer_2, properties {soundWaveAssetPtr=&/HandyMan_Arm_Open_02}}
 {id=WavePlayer_3, properties {soundWaveAssetPtr=&/HandyMan_Arm_Open_03}}
 {id=WavePlayer_4, properties {soundWaveAssetPtr=&/HandyMan_Arm_TargetStart_01}}
 {id=WavePlayer_5, properties {soundWaveAssetPtr=&/HandyMan_Arm_TargetStart_02}}
 {id=WavePlayer_6, properties {soundWaveAssetPtr=&/HandyMan_Arm_TargetStart_03}}
 {id=WavePlayer_7, properties {bLooping=true, soundWaveAssetPtr=&/HandyMan_Arm_TargetLoop_01}}
]
properties {
 attenuationSettings=&Fort_Weapon_Reload_Att
 maxDistance=2250
 soundClassObject=&Fort_WeaponFoley
 volumeMultiplier=0.451
}
