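# Animation Blueprint for the Handyman medium male body cosmetic, adding accessory dynamics and a targeting (ADS) morph with arm open/close sounds.
# It copies the parent mesh pose, constrains holster and clip bones, then simulates RigidBody dynamics into a cached pose.
# A bool blend on bIsTargeting applies one additive morph animation forward or reversed, resetting on activation.
# Targeting events spawn attached sounds, gated by ground or audio-component checks, fading out the opposite sound.
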AnimBlueprint /BRCosmetics/Characters/Player/Male/Medium/Bodies/M_MED_Handyman/Meshes/M_MED_Handyman_AnimBP
target_skeleton=&/M_MED_Handyman_Skeleton
graphs [
 AnimGraph {nodes [
   n1 = "Sequence Player (additive) 'M_MED_Handyman_MorphAnim_Add'"() {
    bLoopAnimation=false
    playRate=-1.3
    sequence=&/M_MED_Handyman_MorphAnim_Add
   }
   n2 = Use cached pose 'startPose'()
   n3 = "Sequence Player (additive) 'M_MED_Handyman_MorphAnim_Add'"() {
    bLoopAnimation=false
    playRate=1.2
    sequence=&M_MED_Handyman_MorphAnim_Add
   }
   n4 = Use cached pose 'startPose'()
   Output Pose(Result=Blend Poses by bool(
     BlendPose_0=Apply Additive(Base=@n4, Additive=@n3) {alphaBoolBlend {blendOption=Linear}}
     BlendPose_1=Apply Additive(Base=@n2, Additive=@n1) {alphaBoolBlend {blendOption=Linear}}
     bActiveValue=Get bIsTargeting()
    ) {blendType=Linear, childUpateMode=ResetChildOnActivate}) {layerGroup=None}
   n5 = "Constraint - dyn_hip_holster_r_MSTR (PP)"!(
    ComponentPose=Local To Component(LocalPose=Copy Pose From Mesh() {bUseAttachedParent=true})
    Alpha=1
    ConstraintWeights_0=0.098
    ConstraintWeights_1:float=0.902
   ) -> (Pose) {
    alphaBoolBlend {blendOption=Linear}
    boneToModify {boneName=dyn_hip_holster_r_MSTR}
    constraintSetup [{targetBone {boneName=pelvis}}, {targetBone {boneName=thigh_twist_01_r}}]
    constraintWeights [1,1]
   }
   n6 = "Constraint - dyn_clips_r_MSTR (PPPP)"!(
    ComponentPose=@n5
    Alpha=1
    ConstraintWeights_0=0.0376
    ConstraintWeights_1:float=0.3177
    ConstraintWeights_2:float=0.62
    ConstraintWeights_3:float=0.0246
   ) -> (Pose) {
    alphaBoolBlend {blendOption=Linear}
    boneToModify {boneName=dyn_clips_r_MSTR}
    constraintSetup [{targetBone {boneName=spine_02}}, {targetBone {boneName=spine_03}}, {targetBone {boneName=spine_04}}, {targetBone {boneName=spine_05}}]
    constraintWeights [1,1,1,1]
   }
   n7 = "Constraint - dyn_clips_l_MSTR (PPP)"!(ComponentPose=@n6, Alpha=1, ConstraintWeights_0=0.1297, ConstraintWeights_1:float=0.4137, ConstraintWeights_2:float=0.4566) -> (Pose) {
    alphaBoolBlend {blendOption=Linear}
    boneToModify {boneName=dyn_clips_l_MSTR}
    constraintSetup [{targetBone {boneName=spine_02}}, {targetBone {boneName=spine_03}}, {targetBone {boneName=spine_04}}]
    constraintWeights [1,1,1]
   }
   n8 = RigidBody(ComponentPose=@n7) {
    bClampLinearTranslationLimitToRefPose=true
    bOverrideWorldGravity=true
    overridePhysicsAsset=&/M_MED_Handyman_Dyn
    simulationSpace=BaseBoneSpace
    alphaBoolBlend {blendOption=Linear}
    baseBoneRef {boneName=spine_04}
    componentAppliedLinearAccClamp=<3000 3000 3000>
    componentLinearAccScale=<0.4 0.5 0.3>
    componentLinearVelScale=<0.4 0.5 0.3>
    overrideWorldGravity {z=-980}
   }
   Save cached pose 'startPose'(Pose=Component To Local(ComponentPose=@n8)) {cacheName=startPose}
  ]}
 DisableADSAudio {nodes [
   n1 = DisableADSAudio!() -> (then, AttachToComponent:Scene Component Object Reference) {is=title K2Node_FunctionEntry}
   > n2 = Get!(execute:Exec) -> (then:Exec, else:Exec, ADSStartAudioComponent:Audio Component Object Reference)
   > FadeOut!(
    execute
    self:Audio Component Object Reference=@n2
    FadeOutDuration:float=0.2
    FadeVolumeLevel:float=0
    FadeCurve:EAudioFaderCurve Enum=Linear
   ) -> (then)
   > n3 = SpawnSoundAttached!(
    execute
    Sound:Sound Base Object Reference=Get ADSEndSound() -> (ADSEndSound:Sound Base Object Reference)
    AttachToComponent:Scene Component Object Reference=@n1
    AttachPointName:Name=None
    Location:Vector
    Rotation:Rotator
    LocationType:EAttachLocation Enum=KeepRelativeOffset
    bStopWhenAttachedToDestroyed:bool=false
    VolumeMultiplier:float=1
    PitchMultiplier:float=1
    StartTime:float=0
    AttenuationSettings:Sound Attenuation Object Reference
    ConcurrencySettings:Sound Concurrency Object Reference
    bAutoDestroy:bool=true
   ) -> (then, ReturnValue:Audio Component Object Reference)
   > Set ADSEndAudioComponent(ADSEndAudioComponent:Audio Component Object Reference=@n3) -> (Output_Get:Audio Component Object Reference)
  ]}
 EnableADSAudio {nodes [
   n1 = EnableADSAudio!() -> (then, AttachToComponent:Scene Component Object Reference) {is=title K2Node_FunctionEntry}
   > n2 = Get()
   n3 = SpawnSoundAttached(
    execute=@n2.else
    Sound=Get ADSStartSound() -> (ADSStartSound:Sound Base Object Reference)
    AttachToComponent=@n1
    AttachPointName=None
    LocationType=KeepRelativeOffset
    bStopWhenAttachedToDestroyed=false
    VolumeMultiplier=1
    PitchMultiplier=1
    StartTime=0
    bAutoDestroy=true
   )
   > Set ADSStartAudioComponent(ADSStartAudioComponent:Audio Component Object Reference=@n3) -> (Output_Get:Audio Component Object Reference)
   > n4 = Get!(execute) -> (then, else, ADSEndAudioComponent:Audio Component Object Reference)
   > FadeOut(self=@n4, FadeOutDuration=0.2, FadeVolumeLevel=0, FadeCurve=Linear)
  ]}
 EventGraph {nodes [
   Event OnBeginTargeting() {is=title K2Node_Event}
   > Parent: OnBeginTargeting() {is=title K2Node_CallParentFunction}
   > Branch(Condition=Get bIsOnGround() -> (bIsOnGround:bool))
   n1 = GetOwnerSkeletalMeshComponent()
   > EnableADSAudio!(execute, self, AttachToComponent:Scene Component Object Reference=@n1) -> (then) {is=title K2Node_CallFunction}
   Event OnEndTargeting() {is=title K2Node_Event}
   > Parent: OnEndTargeting() {is=title K2Node_CallParentFunction}
   n2 = Get ADSStartAudioComponent() -> (ADSStartAudioComponent:Audio Component Object Reference)
   > Branch(Condition=OR Boolean(A=Get bIsOnGround(), B=IsValid(Object=@n2)))
   > DisableADSAudio!(execute, self, AttachToComponent:Scene Component Object Reference=@n1) -> (then) {is=title K2Node_CallFunction}
   Event BlueprintUpdateAnimation()
  ]}
]
variables [
 ADSEndAudioComponent {type=AudioComponent, default=null}
 ADSEndSound {type=SoundBase, default=/BRCosmetics/Sounds/Characters/Handyman/HandyMan_Arm_Close_Cue}
 ADSStartAudioComponent {type=AudioComponent, default=null}
 ADSStartSound {type=SoundBase, default=/BRCosmetics/Sounds/Characters/Handyman/HandyMan_Arm_Open_Cue}
]
