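# Physics asset for the FaceAcc part of the male medium Handyman body, defining bodies for dynamic hair simulation.
# Six three-bone chains hang from the head with limited swing and tight twist; back and left roots swing less than front ones.
# Back and left chains mostly use boxes, front strands thin capsules, alongside head, neck, clavicle and upper-spine capsules.
# Body joints rotate freely, as does an extra cross-link between the back-right and center-back tips.
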PhysicsAsset /BRCosmetics/Characters/Player/Male/Medium/Bodies/M_MED_Handyman/Meshes/Parts/M_MED_Handyman_FaceAcc_Dyn
bodies {
 clavicle_l [{shapeType=Capsule, length=10.755, radius=8.755, center=<6.2465 -0.4655 0.5274>, rotation=<rot 69.9997 0 0>}]
 clavicle_r [{shapeType=Capsule, length=10.755, radius=8.755, center=<-6.2465 0.4655 -0.5274>, rotation=<rot 69.9997 0 0>}]
 dyn_hair_bk_01 [{shapeType=Box, extentX=5.5625, extentY=6.5625, extentZ=1.1875, center=<2.3345 0.0001 0>}]
 dyn_hair_bk_02 [{shapeType=Box, extentX=5.5625, extentY=6.5625, extentZ=1.1875, center=<2.3345 0.0001 0>}]
 dyn_hair_bk_03 [{shapeType=Box, extentX=5.5625, extentY=6.5625, extentZ=1.1875, center=<2.3345 0.0001 0>}]
 dyn_hair_bk_l_01 [{shapeType=Box, extentX=5.5625, extentY=6.5625, extentZ=1.1875, center=<2.3345 0.0001 0>}]
 dyn_hair_bk_l_02 [{shapeType=Box, extentX=5.5625, extentY=6.5625, extentZ=1.1875, center=<2.3345 0.0001 0>}]
 dyn_hair_bk_l_03 [{shapeType=Box, extentX=5.5625, extentY=6.5625, extentZ=1.1875, center=<2.3345 0.0001 0>}]
 dyn_hair_bk_r_01 [{shapeType=Box, extentX=5.5625, extentY=6.5625, extentZ=1.1875, center=<2.3345 0.0001 0>}]
 dyn_hair_bk_r_02 [{shapeType=Box, extentX=5.5625, extentY=6.5625, extentZ=1.1875, center=<2.3345 0.0001 0>}]
 dyn_hair_bk_r_03 [{shapeType=Capsule, length=1.6739, radius=1.7065, center=<2.3897 -0.0001 0>, rotation=<rot 90 0 0>}]
 dyn_hair_fr_l_01 [{shapeType=Capsule, length=4.3438, radius=0.75, shapeName=dyn_hair_fr_l_01_capsule, center=<2.3496 0 -0.0001>, rotation=<rot 90 0 0>}]
 dyn_hair_fr_l_02 [{shapeType=Capsule, length=4.3438, radius=0.75, shapeName=dyn_hair_fr_l_01_capsule, center=<2.3496 0 -0.0001>, rotation=<rot 90 0 0>}]
 dyn_hair_fr_l_03 [{shapeType=Capsule, length=4.3438, radius=0.75, shapeName=dyn_hair_fr_l_01_capsule, center=<2.3496 0 -0.0001>, rotation=<rot 90 0 0>}]
 dyn_hair_fr_r_01 [{shapeType=Capsule, length=4.3438, radius=0.75, shapeName=dyn_hair_fr_l_01_capsule, center=<2.3496 0 -0.0001>, rotation=<rot 90 0 0>}]
 dyn_hair_fr_r_02 [{shapeType=Capsule, length=4.3438, radius=0.75, shapeName=dyn_hair_fr_l_01_capsule, center=<2.3496 0 -0.0001>, rotation=<rot 90 0 0>}]
 dyn_hair_fr_r_03 [{shapeType=Capsule, length=4.3438, radius=0.75, shapeName=dyn_hair_fr_l_01_capsule, center=<2.3496 0 -0.0001>, rotation=<rot 90 0 0>}]
 dyn_hair_l_01 [{shapeType=Box, extentX=5.5625, extentY=6.5625, extentZ=1.1875, center=<2.3345 0.0001 0>}]
 dyn_hair_l_02 [{shapeType=Box, extentX=5.5625, extentY=6.5625, extentZ=1.1875, center=<2.3345 0.0001 0>}]
 dyn_hair_l_03 [{shapeType=Box, extentX=4.764, extentY=6.5625, extentZ=1.1875, center=<2.1012 0.0001 0>}]
 head [
  {shapeType=Capsule, length=3.6644, radius=6.3356, center=<5.1568 1.7703 0>, rotation=<rot 90 0 0>}
  {
   shapeType=Capsule
   length=5.4688
   radius=1.7188
   shapeName=head_capsule
   center=<4.2584 6.1968 -4.6411>
   rotation=<rot 0.2223 2.3785 -95.3363>
  }
 ]
 neck_01 [{shapeType=Capsule, length=9.7147, radius=6.3378, center=<5.4343 0 0>, rotation=<rot 90 0 4.0616>}]
 spine_05 [{shapeType=Capsule, length=4.4231, radius=9.7188, center=<7.4543 0.0001 0>}]
}
constraints [
 {bone1Name=dyn_hair_l_02, bone2Name=dyn_hair_l_01, swing1LimitDegrees=45, swing2LimitDegrees=45, twistLimitDegrees=5}
 {bone1Name=dyn_hair_l_03, bone2Name=dyn_hair_l_02, swing1LimitDegrees=45, swing2LimitDegrees=45, twistLimitDegrees=5}
 {bone1Name=dyn_hair_bk_l_02, bone2Name=dyn_hair_bk_l_01, swing1LimitDegrees=45, swing2LimitDegrees=45, twistLimitDegrees=5}
 {bone1Name=dyn_hair_bk_l_03, bone2Name=dyn_hair_bk_l_02, swing1LimitDegrees=45, swing2LimitDegrees=45, twistLimitDegrees=5}
 {bone1Name=dyn_hair_bk_02, bone2Name=dyn_hair_bk_01, swing1LimitDegrees=45, swing2LimitDegrees=45, twistLimitDegrees=5}
 {bone1Name=dyn_hair_bk_03, bone2Name=dyn_hair_bk_02, swing1LimitDegrees=45, swing2LimitDegrees=45, twistLimitDegrees=5}
 {bone1Name=dyn_hair_fr_l_02, bone2Name=dyn_hair_fr_l_01, swing1LimitDegrees=45, swing2LimitDegrees=45, twistLimitDegrees=5}
 {bone1Name=dyn_hair_fr_l_03, bone2Name=dyn_hair_fr_l_02, swing1LimitDegrees=45, swing2LimitDegrees=45, twistLimitDegrees=5}
 {bone1Name=dyn_hair_bk_r_02, bone2Name=dyn_hair_bk_r_01, swing1LimitDegrees=45, swing2LimitDegrees=45, twistLimitDegrees=5}
 {bone1Name=dyn_hair_bk_r_03, bone2Name=dyn_hair_bk_r_02, swing1LimitDegrees=45, swing2LimitDegrees=45, twistLimitDegrees=5}
 {bone1Name=dyn_hair_fr_r_02, bone2Name=dyn_hair_fr_r_01, swing1LimitDegrees=45, swing2LimitDegrees=45, twistLimitDegrees=5}
 {bone1Name=dyn_hair_fr_r_03, bone2Name=dyn_hair_fr_r_02, swing1LimitDegrees=45, swing2LimitDegrees=45, twistLimitDegrees=5}
 {bone1Name=dyn_hair_l_01, bone2Name=head, swing1LimitDegrees=30, swing2LimitDegrees=20, twistLimitDegrees=5}
 {bone1Name=dyn_hair_bk_l_01, bone2Name=head, swing1LimitDegrees=30, swing2LimitDegrees=20, twistLimitDegrees=5}
 {bone1Name=dyn_hair_bk_01, bone2Name=head, swing1LimitDegrees=30, swing2LimitDegrees=20, twistLimitDegrees=5}
 {bone1Name=dyn_hair_fr_l_01, bone2Name=head, swing1LimitDegrees=45, swing2LimitDegrees=45, twistLimitDegrees=5}
 {bone1Name=dyn_hair_bk_r_01, bone2Name=head, swing1LimitDegrees=30, swing2LimitDegrees=20, twistLimitDegrees=5}
 {bone1Name=dyn_hair_fr_r_01, bone2Name=head, swing1LimitDegrees=45, swing2LimitDegrees=45, twistLimitDegrees=5}
 {
  bone1Name=head
  bone2Name=neck_01
  swing1LimitDegrees=45
  swing1Motion=Free
  swing2LimitDegrees=45
  swing2Motion=Free
  twistLimitDegrees=45
  twistMotion=Free
 }
 {
  bone1Name=clavicle_l
  bone2Name=spine_05
  swing1LimitDegrees=45
  swing1Motion=Free
  swing2LimitDegrees=45
  swing2Motion=Free
  twistLimitDegrees=45
  twistMotion=Free
 }
 {
  bone1Name=clavicle_r
  bone2Name=spine_05
  swing1LimitDegrees=45
  swing1Motion=Free
  swing2LimitDegrees=45
  swing2Motion=Free
  twistLimitDegrees=45
  twistMotion=Free
 }
 {
  bone1Name=neck_01
  bone2Name=spine_05
  swing1LimitDegrees=45
  swing1Motion=Free
  swing2LimitDegrees=45
  swing2Motion=Free
  twistLimitDegrees=45
  twistMotion=Free
 }
 {
  bone1Name=dyn_hair_bk_r_03
  bone2Name=dyn_hair_bk_03
  swing1LimitDegrees=45
  swing1Motion=Free
  swing2LimitDegrees=45
  swing2Motion=Free
  twistLimitDegrees=45
  twistMotion=Free
 }
]
